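# Stereo one-shot MetaSound Source for the wing-flap sounds of the Coconut Shell glider.
# It is a preset of MSS_Glider_Open, filling the parent's Open.Sounds input with seven wing-flap wave variations.
# Overrides set narrow pitch and volume modulation ranges, zero start delay and a reduced template volume.
# Although built on the Open template, it uses glider thrust attenuation and concurrency, with the Fort_Gliders_Land sound class.
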MetaSoundSource /BRCosmetics/Sounds/Gliders/CoconutShell/MSS_Glider_CoconutShell_WingFlaps
preset_parent=/FortMetasound/Cosmetics/Glider/MSS_Glider_Open
inputs [
 Open.Sounds {type=WaveAsset:Array}
 PitchModulationRange.Max {type=Float}
 PitchModulationRange.Min {type=Float}
 StartDelay.Max {type=Time}
 StartDelay.Min {type=Time}
 TemplatePitchMultiplier {type=Float}
 TemplateVolumeMultiplier {type=Float}
 UE.Source.OnPlay {type=Trigger}
 VolumeModulationRange.Max {type=Float}
 VolumeModulationRange.Min {type=Float}
]
interfaces [UE.OutputFormat.Stereo:1.0, UE.Source.OneShot:1.0, UE.Source:1.1]
outputs [
 "UE.OutputFormat.Stereo.Audio:0" {type=Audio}
 "UE.OutputFormat.Stereo.Audio:1" {type=Audio}
 UE.Source.OneShot.OnFinished {type=Trigger}
]
pages [Default {input_defaults {
   PitchModulationRange.Max=1.02
   PitchModulationRange.Min=0.98
   StartDelay.Max=0
   StartDelay.Min=0
   TemplatePitchMultiplier=1
   TemplateVolumeMultiplier=0.7
   UE.Source.OnPlay=null
   VolumeModulationRange.Max=1.02
   VolumeModulationRange.Min=0.98
   Open.Sounds [
    &/IP_Glider_CoconutShell_Wingflap_01
    &/IP_Glider_CoconutShell_Wingflap_02
    &/IP_Glider_CoconutShell_Wingflap_03
    &/IP_Glider_CoconutShell_Wingflap_04
    &/IP_Glider_CoconutShell_Wingflap_05
    &/IP_Glider_CoconutShell_Wingflap_06
    &/IP_Glider_CoconutShell_Wingflap_07
   ]
  }}]
properties {
 assetImportData=&self:AssetImportData
 attenuationSettings=&Fort_Player_Glider_Thrust_Athena_Att
 maxDistance=12500
 numChannels=2
 outputFormat=Stereo
 sampleRate=48000
 soundClassObject=&Fort_Gliders_Land
 concurrencySet [&Athena_Glider_Thrust_Conc]
}
